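# Additive unlit material for red light rays in the Helios environment, shaped and faded around two world-space center points.
# Opacity terms combine UV falloff, inverse Fresnel, nearest-center fade, depth fade and a camera-height fade; a panning god-ray mask forms a separate branch.
# An offset term stretches vertices vertically by UV and center distance, collapsing them to the object pivot when the camera is below the height threshold.
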Material /Game/Environments/Helios/TechArt/HadesRays/M_HadesRays
graph {
 properties {blendMode=BLEND_Additive, pixelDepthOffsetMode=PDOM_Legacy, shadingModel=MSM_Unlit, usedShadingModels=Unlit}
 nodes [
  Constant3Vector()
  n1 = OneMinus(None=ComponentMask(None=TextureCoordinate()) {g=true})
  n2 = Constant3Vector() {constant=<rgba 0 0 1 0>}
  n3 = ScalarParameter() {defaultValue=9276.4014, parameterName=OffsetCenter}
  n4 = ScalarParameter() {defaultValue=5509.9443, parameterName=Offset}
  n5 = Distance(A=VectorParameter() {parameterName=Center, defaultValue=<rgba 44691.7422 -79466.7188 0 1>}.RGB, B=WorldPosition().XYZ)
  n6 = Distance(A=VectorParameter() {parameterName=Center2, defaultValue=<rgba 20785.9492 -58805.8711 0 1>}.RGB, B=WorldPosition().XYZ)
  n7 = Min(A=@n5, B=@n6)
  n8 = ScalarParameter() {defaultValue=12335.1436, parameterName=CenterFadeWPO}
  n9 = Saturate(None=Power(Base=Divide(A=@n7, B=@n8), Exp=ScalarParameter() {defaultValue=525006.3125, parameterName=CenterFadeWPOFalloff}))
  n10 = ComponentMask(None=CameraPositionWS()) {b=true}
  n11 = ScalarParameter() {defaultValue=13586.2656, parameterName=HeightFade_Offset}
  n12 = Saturate(None=Divide(A=Subtract(A=@n10, B=@n11), B=ScalarParameter() {defaultValue=3017.3462, parameterName=HeightFade_Distance}))
  LinearInterpolate(
   A=Subtract(A=ObjectPositionWS(), B=WorldPosition().XYZ)
   B=Multiply(A=Multiply(A=@n1, B=@n2), B=LinearInterpolate(A=@n3, B=@n4, Alpha=@n9))
   Alpha=Ceil(None=@n12)
  )
  Max(A=@n5, B=@n6)
  Multiply(
   A=ScalarParameter() {defaultValue=0.232, parameterName=Brightness}
   B=VectorParameter() {parameterName=Color, defaultValue=<rgba 1 0.0256 0 1>}.RGB
  )
  n13 = Multiply(A=TextureCoordinate() {coordinateIndex=1}, B=ScalarParameter() {defaultValue=-2.0449, parameterName=UV_Scale})
  n14 = ScalarParameter() {defaultValue=0.028, parameterName=NoisePanSpeed}
  n15 = ComponentMask(None=TextureCoordinate()) {g=true}
  n16 = ScalarParameter() {defaultValue=0.2885, parameterName=Falloff}
  n17 = ScalarParameter() {defaultValue=0.1457, parameterName=FresnellPower}
  n18 = Fresnel(ExponentIn=@n17) {baseReflectFraction=0}
  n19 = Power(
   Base=Divide(A=@n7, B=ScalarParameter() {defaultValue=13050.9268, parameterName=CenterFade})
   Exp=ScalarParameter() {defaultValue=1.4561, parameterName=CenterFadeFalloff}
  )
  n20 = Multiply(
   A=Multiply(A=Power(Base=@n15, Exp=@n16), B=OneMinus(None=@n18))
   B=LinearInterpolate(B=ScalarParameter() {defaultValue=0.08, parameterName=CenterFade_Brightness}, Alpha=Saturate(None=@n19)) {constA=1}
  )
  Multiply(A=TextureSample(UVs=Panner(Coordinate=@n13, Speed=@n14)) {samplerType=SAMPLERTYPE_Masks, texture=/Game/ArtTools/GodRays/GodRays_M}.G, B=@n20)
  Multiply(A=DepthFade(Opacity=@n20, FadeDistance=ScalarParameter() {defaultValue=2793.6479, parameterName=DepthFade}), B=@n12)
  n21 = Constant3Vector() {constant=<rgba 1 1 0 0>}
  n22 = Multiply(A=@n21, B=VertexNormalWS())
  n23 = Multiply(A=@n21, B=CameraVectorWS())
  Power(
   Base=DotProduct(
    A=MaterialFunctionCall(Vector=@n22) {materialFunction=&SafeNormalize}.Result
    B=MaterialFunctionCall(Vector=@n23) {materialFunction=&SafeNormalize}.Result
   )
   Exp=@n17
  )
 ]
}
