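# Additive unlit glow material for Hades rays in the Helios environment, with an orange-red color parameter and panning cloud noise.
# Vertex-red edge falloff, remapped noise and depth fade shape its opacity, scaled by a camera-height fade.
# An offset toward the object pivot, gated by that fade's ceiling, collapses the mesh when the camera is below the threshold.
# A time-driven, view-based parallax UV branch feeds no texture sample.
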Material /Game/Environments/Helios/TechArt/HadesRays/M_Hades_Glow
graph {
 properties {blendMode=BLEND_Additive, pixelDepthOffsetMode=PDOM_Legacy, shadingModel=MSM_Unlit, usedShadingModels=Unlit}
 nodes [
  n1 = Multiply(A=TextureCoordinate(), B=ScalarParameter() {defaultValue=0.5758, parameterName=CloudTiling})
  n2 = CameraVectorWS()
  n3 = Constant3Vector() {constant=<rgba 1 1 0 0>}
  n4 = ComponentMask(None=MaterialFunctionCall(Vector=Multiply(A=@n2, B=@n3)) {materialFunction=&SafeNormalize}.Result) {g=true, r=true}
  n5 = Time()
  n6 = ScalarParameter() {defaultValue=0.004, parameterName=Speed}
  n7 = LinearInterpolate(A=ScalarParameter() {defaultValue=-0.024, parameterName=Displacment}, Alpha=Frac(None=Multiply(A=@n5, B=@n6))) {constB=0}
  n8 = VertexNormalWS()
  Add(A=@n1, B=Divide(A=Multiply(A=@n4, B=@n7), B=DotProduct(A=@n8, B=@n2)))
  n9 = ComponentMask(None=CameraPositionWS()) {b=true}
  n10 = ScalarParameter() {defaultValue=13586.2656, parameterName=HeightFade_Offset}
  n11 = Saturate(None=Divide(A=Subtract(A=@n9, B=@n10), B=ScalarParameter() {defaultValue=3017.3462, parameterName=HeightFade_Distance}))
  LinearInterpolate(A=Subtract(A=ObjectPositionWS(), B=WorldPosition().XYZ), B=Constant3Vector(), Alpha=Ceil(None=@n11))
  n12 = Multiply(
   A=ScalarParameter() {defaultValue=0.288, parameterName=Opacity}
   B=Power(Base=VertexColor().R, Exp=ScalarParameter() {defaultValue=1.2511, parameterName=FalloffEdges})
  )
  n13 = ScalarParameter() {defaultValue=0.3429, parameterName=CloudFalloff_low}
  n14 = TextureSample(UVs=Panner(Coordinate=@n1, Speed=@n6)) {texture=&AtmosphericCloudNoise01}
  n15 = Saturate(None=LinearInterpolate(A=@n13, B=ScalarParameter() {defaultValue=0.3774, parameterName=CloudFalloff_High}, Alpha=@n14.R))
  Multiply(A=DepthFade(Opacity=Multiply(A=@n12, B=@n15), FadeDistance=ScalarParameter() {defaultValue=5599.9468, parameterName=DepthFade}), B=@n11)
  Power(Base=@n14.R, Exp=@n13)
  VectorParameter() {parameterName=Color, defaultValue=<rgba 1.1 0.0464 0 1>}
 ]
}
